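# Translucent UI font material, named for the reward count text in the Habanero lobby frontend.
# It offsets texture coordinates across the text field, scales them around their center and feeds them into a radial gradient function.
# The inverted smooth gradient multiplies vertex alpha, creating an off-center soft radial mask across the text.
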Material /FortniteCoreUI/UI/Frontend/Lobby/Habanero/Materials/M_UI_Font_Lobby_RewardCount
graph {
 properties {
  bAutomaticallySetUsageInEditor=false
  bUsedWithStaticMesh=false
  blendMode=BLEND_Translucent
  materialDomain=MD_UI
  usedShadingModels=Default Lit
 }
 nodes [
  n1 = MaterialFunctionCall(UVs=Add(A=TextureCoordinate() {desc=UV across Textfield}, B=Constant2Vector() {g=-0.5, r=0.375}), Texture Scale=Constant() {r=2.5}) {materialFunction=&ScaleUVsByCenter}
  Multiply(
   A=VertexColor().A
   B=OneMinus(None=MaterialFunctionCall(UV=@n1.UVs) {materialFunction=&MF_UI_RadialGradient}.Smooth)
  )
 ]
}
