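# UI material that draws a procedural, animated flare, named for the lobby level text, using alpha composite blending.
# Mirrored UV gradients with power and smoothstep falloffs shape the flare mask without textures.
# A time-rotated UV blends white with the CoreUI oscillating color function, serving as the exponent on a purple constant, plus a radial center glow.
# Vertex color, vertex alpha and the FlareOpacity parameter modulate the result.
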Material /FortniteCoreUI/UI/Frontend/Lobby/Habanero/Materials/M_UI_Lobby_Level_TextFlare
graph {
 properties {
  bAutomaticallySetUsageInEditor=false
  bUsedWithStaticMesh=false
  blendMode=BLEND_AlphaComposite
  materialDomain=MD_UI
  usedShadingModels=Default Lit
 }
 nodes [
  n1 = VertexColor()
  n2 = MaterialFunctionCall() {materialFunction=&GetUserInterfaceUV}
  n3 = ComponentMask(None=@n2.Normalized UV) {r=true}
  n4 = Multiply(A=Min(A=@n3, B=OneMinus(None=@n3))) {constB=2}
  n5 = ComponentMask(None=@n2.Normalized UV) {g=true}
  n6 = Multiply(A=Min(A=@n5, B=OneMinus(None=@n5))) {constB=2}
  n7 = Power(Base=Multiply(A=@n4, B=Power(Base=@n6) {constExponent=6}))
  n8 = Power(Base=SmoothStep(Value=@n4)) {constExponent=4}
  n9 = LinearInterpolate(A=Multiply(A=Multiply(A=@n7) {constB=0.5}, B=OneMinus(None=@n8)), Alpha=Multiply(A=@n8) {constB=0.2})
  Multiply(A=@n1.A, B=@n9)
  n10 = Constant3Vector() {constant=<rgba 0.4501 0.1094 1 1>}
  n11 = Rotator(
   Coordinate=MaterialFunctionCall() {materialFunction=&GetUserInterfaceUV}.Normalized UV
   Time=Multiply(A=Time()) {constB=0.333}
  ) {speed=6.2832}
  n12 = NamedRerouteDeclaration(None=LinearInterpolate(
    A=Constant3Vector() {constant=<rgba 1 1 1 1>}
    B=MaterialFunctionCall!(PrimaryUV, SecondaryUV, UVMix, PixelSize) {materialFunction=/FortUILibrary/MaterialFunctions/CoreUI/MF_UI_CoreUI_OscillatingColor}.Result
    Alpha=ComponentMask(None=@n11) {r=true}
   ) {constA=1}) {name=RiftOscillatingColors, nodeColor=<rgba 0 0 0 1>}
  n13 = ConstantBiasScale(None=@n2.Normalized UV) {bias=-0.5, scale=2}
  n14 = OneMinus(None=Saturate(None=Length(None=@n13) {is=class MaterialExpressionLength}))
  n15 = Saturate(None=Add(A=Power(Base=@n10, Exp=@n12), B=Multiply(A=@n14) {constB=0.5}))
  n16 = Multiply(A=ScalarParameter() {defaultValue=1, parameterName=FlareOpacity, sliderMax=1}, B=@n9)
  Multiply(A=@n1, B=Multiply(A=@n1.A, B=Multiply(A=@n15, B=@n16)))
 ]
}
