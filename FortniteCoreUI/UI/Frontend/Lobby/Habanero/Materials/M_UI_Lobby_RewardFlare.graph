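# UI material for the lobby reward flare, an alpha-composite effect of glow, flare and drifting triangle particles.
# It builds named masks for an offsettable radial glow, a horizontally scaled, range-limited noisy flare, wave noise and rotated, panning viewport noise.
# Triangles come from a screen-space particle grid with randomized speed, size and visibility.
# Masks are blended with opacity parameters, two-color tinting and vertex color, alongside an oscillating rift color chain.
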Material /FortniteCoreUI/UI/Frontend/Lobby/Habanero/Materials/M_UI_Lobby_RewardFlare
graph {
 properties {
  bUsedWithStaticMesh=false
  blendMode=BLEND_AlphaComposite
  materialDomain=MD_UI
  pixelDepthOffsetMode=PDOM_Legacy
  usedShadingModels=Default Lit
 }
 nodes [
  n1 = MaterialFunctionCall() {materialFunction=&GetUserInterfaceUV}
  MaterialFunctionCall(UV=@n1.Normalized UV) {materialFunction=&MF_UI_UVScaling}
  n2 = VertexColor()
  n3 = LinearInterpolate(
   A=Multiply(A=NamedRerouteUsage(), B=ScalarParameter() {defaultValue=1, parameterName=Glow Opacity, sliderMax=1})
   Alpha=Saturate(None=NamedRerouteUsage())
  )
  n4 = Multiply(A=Multiply(A=NamedRerouteUsage(), B=ScalarParameter() {defaultValue=1, parameterName=Triangles Opacity, sliderMax=1}), B=@n3)
  n5 = Multiply(A=@n3, B=ScalarParameter() {defaultValue=1, parameterName=Flare Opacity, sliderMax=1})
  Multiply(A=@n2.A, B=Multiply(A=LinearInterpolate(A=@n4, Alpha=@n5), B=ScalarParameter() {defaultValue=1, parameterName=Global Opacity, sliderMax=1}))
  n6 = LinearInterpolate(
   A=VectorParameter() {parameterName=Color B, defaultValue=<rgba 0.0999 0.0212 1 1>}.RGB
   B=VectorParameter() {parameterName=Color A, defaultValue=<rgba 1 0 0.2793 1>}.RGB
   Alpha=NamedRerouteUsage()
  )
  n7 = NamedRerouteUsage()
  n8 = ScalarParameter() {defaultValue=1, parameterName=Glow Intensity, sliderMax=1}
  n9 = Saturate(None=Subtract(A=@n7, B=OneMinus(None=@n8)) {constB=0})
  n10 = Multiply(A=Multiply(A=@n6, B=@n7)) {constB=2}
  n11 = Add(A=@n6, B=Add(A=Multiply(A=@n9) {constB=2}, B=Add(A=@n10, B=@n10)))
  Multiply(A=Multiply(A=Saturate(None=@n11), B=@n2), B=@n2.A)
  n12 = VertexColor()
  n13 = NamedRerouteUsage()
  n14 = SmoothStep(Value=Multiply(A=@n13, B=@n13))
  n15 = Saturate(None=Add(A=NamedRerouteUsage(), B=Multiply(A=@n14)))
  Multiply(A=@n12.A, B=@n15)
  n16 = ScalarParameter() {parameterName=Rotation, sliderMax=1}
  n17 = MaterialFunctionCall(UVs=MaterialFunctionCall() {materialFunction=&GetUserInterfaceUV}.Normalized UV, "Rotation Angle (0-1)"=@n16) {materialFunction=&CustomRotator}
  n18 = Subtract(A=@n17.Rotated Values, B=Constant2Vector() {g=-0.5})
  n19 = Length(None=ConstantBiasScale(None=@n18) {bias=-0.5, scale=2}) {is=class MaterialExpressionLength}
  n20 = SmoothStep(Value=OneMinus(None=Saturate(None=@n19)))
  n21 = Saturate(None=Add(A=Multiply(A=@n20) {constB=0.75}, B=NamedRerouteUsage()))
  Multiply(A=@n12, B=Multiply(A=@n12.A, B=Multiply(A=@n21, B=@n15)))
  n22 = MaterialFunctionCall(UVs=@n1.Normalized UV, Texture Scale=AppendVector(A=ScalarParameter() {defaultValue=0.25, parameterName=Flare Scale X}, B=Constant() {r=1})) {materialFunction=&ScaleUVsByCenter}
  n23 = MaterialFunctionCall(UV Channel=@n22.UVs) {materialFunction=&LinearGradient}
  n24 = Multiply(A=Min(A=OneMinus(None=@n23.UGradient), B=@n23.UGradient)) {constB=2}
  Multiply(A=MaterialFunctionCall("Value (0-1)"=@n24) {materialFunction=&MF_UI_EaseCurves}.EaseOutQuint)
  n25 = Multiply(A=Time()) {constB=0.1}
  n26 = Add(
   A=MaterialFunctionCall() {materialFunction=&GetUserInterfaceUV}.Normalized UV
   B=AppendVector(A=Frac(None=@n25), B=Constant())
  )
  n27 = TextureObject() {samplerType=SAMPLERTYPE_Grayscale, texture=/FortUILibrary/UIKit/Buttons/Textures/LowResBlurredNoise}
  n28 = MaterialFunctionCall!(
   UVs=@n22.UVs
   Alpha=MaterialFunctionCall("Value (0-1)"=@n24) {materialFunction=&MF_UI_EaseCurves}.EaseOutSine
  ) {materialFunction=/FortUILibrary/MaterialFunctions/Basics/MF_UI_ClipEdgePixel}
  n29 = Multiply(A=@n28.Result, B=LinearInterpolate(Alpha=TextureSample(UVs=@n26, Tex=@n27).RGB) {constA=0.5})
  n30 = OneMinus(None=@n23.VGradient)
  n31 = OneMinus(None=ScalarParameter() {defaultValue=1, parameterName=Flare Range, sliderMax=1})
  n32 = MaterialFunctionCall("Value (0-1)"=Saturate(None=Subtract(A=@n30, B=@n31))) {materialFunction=&MF_UI_EaseCurves}
  n33 = ScalarParameter() {defaultValue=1, parameterName=Flare Opacity, sliderMax=1}
  NamedRerouteDeclaration(None=Multiply(A=Multiply(A=@n29, B=@n32.EaseOutSine), B=@n33)) {name=Flare Noise Mask, nodeColor=<rgba 0 0 0 1>}
  n34 = MaterialFunctionCall() {materialFunction=&GetUserInterfaceUV}
  n35 = AppendVector(A=ScalarParameter() {parameterName=OffsetX}, B=ScalarParameter() {defaultValue=0.5, parameterName=OffsetY})
  n36 = ScalarParameter() {defaultValue=1, parameterName=ScaleX}
  n37 = ScalarParameter() {defaultValue=1, parameterName=ScaleY}
  n38 = MaterialFunctionCall(
   UV=MaterialFunctionCall(UVs=Add(A=@n34.Normalized UV, B=@n35), Texture Scale=AppendVector(A=@n36, B=@n37)) {materialFunction=&ScaleUVsByCenter}.UVs
   TextureSize=@n34.Pixel Size
  ) {materialFunction=&MF_UI_UVScaling}
  NamedRerouteDeclaration(None=MaterialFunctionCall("Value (0-1)"=MaterialFunctionCall(UV=@n38.Scaled to Fit) {materialFunction=&MF_UI_RadialGradient}.Smooth) {materialFunction=&MF_UI_EaseCurves}.EaseOutQuint) {name=Radial Glow Mask, nodeColor=<rgba 0 0 0 1>}
  n39 = MaterialFunctionCall() {materialFunction=&GetUserInterfaceUV}
  n40 = Multiply(A=Time()) {constB=0.333}
  n41 = LinearInterpolate(
   A=Constant3Vector() {constant=<rgba 1 1 1 1>}
   B=MaterialFunctionCall!(PrimaryUV, SecondaryUV, UVMix, PixelSize) {materialFunction=/FortUILibrary/MaterialFunctions/CoreUI/MF_UI_CoreUI_OscillatingColor}.Result
   Alpha=ComponentMask(None=Rotator(Coordinate=@n39.Normalized UV, Time=@n40) {speed=6.2832}) {r=true}
  ) {constA=1}
  NamedRerouteDeclaration(None=Power(
    Base=Constant3Vector() {constant=<rgba 0.4501 0.1094 1 1>}
    Exp=NamedRerouteDeclaration(None=@n41) {name=RiftOscillatingColors, nodeColor=<rgba 0 0 0 1>}
   )) {name=Coolor, nodeColor=<rgba 0 0 0 1>}
  n42 = ComponentMask(None=@n17.Rotated Values) {r=true}
  n43 = Min(A=@n42, B=OneMinus(None=@n42))
  n44 = NamedRerouteDeclaration(None=SmoothStep(Value=Multiply(A=@n43) {constB=2})) {name=Reflected Gradient Mask, nodeColor=<rgba 0 0 0 1>}
  n45 = OneMinus(None=ComponentMask(None=@n17.Rotated Values) {g=true})
  n46 = Add(
   A=@n17.Rotated Values
   B=AppendVector(
    A=Constant()
    B=MaterialFunctionCall!(
     UV=@n17.Rotated Values
     Wave Height
     Wave Elevation
     Wave Length
     Wave Speed
     GlowMin
     GlowMax
     Outline Thickness
     Outline Alignment
     Custom Curve
    ) {materialFunction=/FortUILibrary/MaterialFunctions/Shapes/MF_UI_Wave}.SDF
   )
  )
  n47 = Constant2Vector() {g=32, r=0.5}
  n48 = TextureObject() {samplerType=SAMPLERTYPE_Grayscale, texture=/Game/Global/Textures/Default/Noise/LowResBlurredNoise}
  n49 = Multiply(
   A=Power(Base=@n45) {constExponent=4}
   B=TextureSample(
    UVs=MaterialFunctionCall(UVs=@n46, Texture Scale=@n47) {materialFunction=&ScaleUVsByCenter}.UVs
    Tex=@n48
   ) {samplerType=SAMPLERTYPE_Grayscale}.R
  )
  n50 = SmoothStep(Value=@n45)
  n51 = Constant2Vector() {g=32, r=1}
  n52 = Multiply(
   A=TextureSample(
    UVs=MaterialFunctionCall(UVs=@n46, Texture Scale=@n51) {materialFunction=&ScaleUVsByCenter}.UVs
    Tex=@n48
   ) {samplerType=SAMPLERTYPE_Grayscale}.R
   B=Power(Base=NamedRerouteUsage()) {constExponent=4}
  )
  NamedRerouteDeclaration(None=Add(A=Multiply(A=@n44, B=@n49), B=Multiply(A=@n50, B=@n52))) {name=Wave Noise, nodeColor=<rgba 0 0 0 1>}
  NamedRerouteDeclaration(None=Multiply(A=@n33, B=Multiply(A=@n28.Result, B=@n32.EaseOutSine))) {name=Flare Mask, nodeColor=<rgba 0 0 0 1>}
  n53 = Time()
  n54 = NamedRerouteDeclaration(None=Multiply(A=Multiply(A=@n53) {constB=-1}, B=ScalarParameter() {defaultValue=0.125, parameterName=Particle Speed}) {constB=-1}) {name=ParticleSpeed, nodeColor=<rgba 0.6406 1 0 1>}
  n55 = Multiply(
   A=NamedRerouteUsage()
   B=NamedRerouteDeclaration(None=ScalarParameter() {defaultValue=64, parameterName=Particle Tiling}) {name=Tiling, nodeColor=<rgba 0 0 0 1>}
  )
  n56 = MaterialFunctionCall!(UV=ComponentMask(None=Floor(None=@n55)) {r=true}, Seed=Constant() {r=0.0123}) {materialFunction=&MF_UI_Random}
  n57 = MaterialFunctionCall("Value (0-1)"=LinearInterpolate(Alpha=Saturate(None=@n56.Result)) {constA=0.75}) {materialFunction=&MF_UI_EaseCurves}
  n58 = Time()
  n59 = ScalarParameter() {parameterName=Triangle Fade Remover, sliderMax=1}
  n60 = LinearInterpolate(
   B=ComponentMask(None=MaterialFunctionCall() {materialFunction=&GetUserInterfaceUV}.Normalized UV) {g=true}
   Alpha=@n59
  ) {constA=1}
  MaterialExpressionRemap_0 = MaterialXRemap!(None=Saturate(None=Subtract(A=@n60, B=@n59) {constB=0.5}), InputLow, InputHigh=OneMinus(None=@n59), TargetLow, TargetHigh)
  n61 = MaterialFunctionCall("Value (0-1)"=@MaterialExpressionRemap_0) {materialFunction=&MF_UI_EaseCurves}
  n62 = NamedRerouteUsage()
  n63 = Multiply(
   A=Multiply(A=ScalarParameter() {defaultValue=1, parameterName=Particle Scale, sliderMax=1}, B=Multiply(A=@n61.EaseOutQuint, B=@n62))
   B=ScalarParameter() {defaultValue=1, parameterName=Triangle Remover, sliderMax=1}
  )
  n64 = LinearInterpolate(A=NamedRerouteUsage(), Alpha=NamedRerouteUsage())
  n65 = MaterialFunctionCall!(
   UV=NamedRerouteDeclaration(None=MaterialFunctionCall!(ScreenUV, TextureSize, ScalePivotPoint) {materialFunction=/FortUILibrary/MaterialFunctions/Basics/MF_UI_ScreenSpaceUVScaling}.Scaled to Fit) {name=Screenspace UVs, nodeColor=<rgba 0 0 0 1>}
   Global Offset=AppendVector(A=Constant(), B=Multiply(A=@n54, B=@n57.EaseInQuint))
   Global Rotation
   Tiling=NamedRerouteUsage()
   Cell Offset=Constant()
   Cell Scale=Max(A=Multiply(A=@n63, B=@n64)) {constB=0}
   Cell Rotation=Add(A=Multiply(A=@n58) {constB=0.05}, B=NamedRerouteUsage())
   Random Seed=Constant() {r=0.0123}
  ) {materialFunction=/FortUILibrary/MaterialFunctions/Basics/MF_UI_ParticleGrid}
  n66 = MaterialFunctionCall("Value (0-1)"=@n65.Random Cell Values) {materialFunction=&MF_UI_EaseCurves}
  n67 = NamedRerouteDeclaration(None=Saturate(None=@n66.EaseInCubic)) {name=Random Cells, nodeColor=<rgba 0 0 0 1>}
  n68 = MaterialFunctionCall!(
   UVs=MaterialFunctionCall(UVs=@n65.Cells UV, Texture Scale=Multiply(A=@n67) {constB=0.5}) {materialFunction=&ScaleUVsByCenter}.UVs
   Corner 01=Constant2Vector() {g=-1, r=-1}
   Corner 02=Constant2Vector() {g=-1, r=1}
   Corner 03
   Glow Min
   Glow Max=Multiply(A=@n67, B=Constant() {r=0.5}) {constB=0.9}
   Outline Thickness
  ) {materialFunction=/FortUILibrary/MaterialFunctions/Shapes/MF_UI_WIP_Triangle}
  n69 = Ceil(None=Saturate(None=LinearInterpolate(Alpha=@n66.EaseInQuint) {constA=-2}))
  NamedRerouteDeclaration(None=Saturate(None=DotProduct(A=@n68.Fill, B=@n69))) {name=Triangles, nodeColor=<rgba 0 0 0 1>}
  n70 = Constant2Vector() {g=0.25, r=1}
  n71 = Length(None=ConstantBiasScale(None=MaterialFunctionCall(UVs=@n18, Texture Scale=@n70) {materialFunction=&ScaleUVsByCenter}.UVs) {bias=-0.5, scale=2})
  NamedRerouteDeclaration(None=OneMinus(None=Saturate(None=@n71))) {name=Radial Mask, nodeColor=<rgba 0 0 0 1>}
  n72 = MaterialFunctionCall(UVs=ScreenPosition().ViewportUV, "Rotation Angle (0-1)"=@n16) {materialFunction=&CustomRotator}
  n73 = Constant() {r=0.375}
  n74 = Panner(Coordinate=Rotator(Coordinate=MaterialFunctionCall(UVs=@n72.Rotated Values, Texture Scale=@n73) {materialFunction=&ScaleUVsByCenter}.UVs) {speed=-0.125}) {speedX=-0.05}
  n75 = TextureObject() {samplerType=SAMPLERTYPE_Grayscale, texture=&LowResBlurredNoise}
  n76 = Panner(Coordinate=MaterialFunctionCall(UVs=@n72.Rotated Values, Texture Scale=Constant() {r=3}) {materialFunction=&ScaleUVsByCenter}.UVs) {speedX=0.025, speedY=0.1}
  n77 = Multiply(A=Multiply(
    A=TextureSample(UVs=@n74, Tex=@n75) {samplerType=SAMPLERTYPE_Grayscale}.R
    B=TextureSample(UVs=@n76, Tex=@n75) {samplerType=SAMPLERTYPE_Grayscale}.R
   )) {constB=2}
  NamedRerouteDeclaration(None=Power(Base=Saturate(None=@n77))) {name=Noise Mask, nodeColor=<rgba 0 0 0 1>}
 ]
}
