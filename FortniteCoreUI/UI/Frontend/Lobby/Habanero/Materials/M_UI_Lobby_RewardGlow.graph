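# UI material for a soft radial reward glow in the Habanero lobby, using alpha-composite blending.
# A time-driven UV rotator blends white with the CoreUI oscillating color, and the result is the exponent on a purple constant, giving shifting hues.
# A smooth radial gradient and vertex color shape the glow.
# The Opacity parameter scales the tinted branch, while a separate alpha branch uses a fixed gradient multiplier.
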Material /FortniteCoreUI/UI/Frontend/Lobby/Habanero/Materials/M_UI_Lobby_RewardGlow
graph {
 properties {
  bAutomaticallySetUsageInEditor=false
  bUsedWithStaticMesh=false
  blendMode=BLEND_AlphaComposite
  materialDomain=MD_UI
  usedShadingModels=Default Lit
 }
 nodes [
  MaterialFunctionCall(
   UVs=Add(A=MaterialFunctionCall() {materialFunction=&MF_UI_UVScaling}.ScaleX, B=Constant2Vector() {r=0.0625})
   Texture Scale=Constant() {r=0.85}
  ) {materialFunction=&ScaleUVsByCenter}
  n1 = VertexColor()
  n2 = MaterialFunctionCall() {materialFunction=&MF_UI_RadialGradient}
  Multiply(A=@n1.A, B=Multiply(A=@n2.Smooth) {constB=0.125})
  n3 = MaterialFunctionCall() {materialFunction=&GetUserInterfaceUV}
  n4 = Multiply(A=Time()) {constB=0.333}
  n5 = LinearInterpolate(
   A=Constant3Vector() {constant=<rgba 1 1 1 1>}
   B=MaterialFunctionCall!(PrimaryUV, SecondaryUV, UVMix, PixelSize) {materialFunction=/FortUILibrary/MaterialFunctions/CoreUI/MF_UI_CoreUI_OscillatingColor}.Result
   Alpha=ComponentMask(None=Rotator(Coordinate=@n3.Normalized UV, Time=@n4) {speed=6.2832}) {r=true}
  ) {constA=1}
  n6 = Saturate(None=Power(
    Base=Constant3Vector() {constant=<rgba 0.4501 0.1094 1 1>}
    Exp=NamedRerouteDeclaration(None=@n5) {name=RiftOscillatingColors, nodeColor=<rgba 0 0 0 1>}
   ))
  n7 = Multiply(A=@n2.Smooth, B=ScalarParameter() {defaultValue=0.1, parameterName=Opacity, sliderMax=1}) {constB=0.1}
  Multiply(A=@n1, B=Multiply(A=@n1.A, B=Multiply(A=@n6, B=@n7)))
 ]
}
