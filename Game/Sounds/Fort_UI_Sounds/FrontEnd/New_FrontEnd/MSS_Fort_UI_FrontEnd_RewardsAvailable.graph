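# One-shot 5.1 MetaSound Source for the front-end UI "rewards available" cue, on the Fort_UI sound class.
# On play, a single 5.1 wave player plays the Fort_UI_RewardEarned_03 asset, and each of its six channels is scaled by the MasterVolume input before output.
# The source finishes when the wave ends, and a concurrency override limits it to one playing instance.
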MetaSoundSource /Game/Sounds/Fort_UI_Sounds/FrontEnd/New_FrontEnd/MSS_Fort_UI_FrontEnd_RewardsAvailable
inputs [
 MasterVolume {type=Float}
 UE.Source.OnPlay {type=Trigger}
]
interfaces [UE.OutputFormat.5dot1:1.0, UE.Source.OneShot:1.0, UE.Source:1.1]
outputs [
 "UE.OutputFormat.5dot1.Audio:0" {type=Audio}
 "UE.OutputFormat.5dot1.Audio:1" {type=Audio}
 "UE.OutputFormat.5dot1.Audio:2" {type=Audio}
 "UE.OutputFormat.5dot1.Audio:3" {type=Audio}
 "UE.OutputFormat.5dot1.Audio:4" {type=Audio}
 "UE.OutputFormat.5dot1.Audio:5" {type=Audio}
 UE.Source.OneShot.OnFinished {type=Trigger}
]
pages [Default {
  input_defaults {MasterVolume=0.7079, UE.Source.OnPlay=null}
  nodes [
   WavePlayer_0 = UE.Wave Player:5dot1(Play=^UE.Source.OnPlay) {also {defaults {Wave Asset=&/Fort_UI_RewardEarned_03}, version [1,1]}}
   Multiply_0 = UE.Multiply:Audio by Float(AdditionalOperands=^MasterVolume, PrimaryOperand=@WavePlayer_0.Out Front Left)
   Multiply_1 = UE.Multiply:Audio by Float(AdditionalOperands=^MasterVolume, PrimaryOperand=@WavePlayer_0.Out Front Center)
   Multiply_2 = UE.Multiply:Audio by Float(AdditionalOperands=^MasterVolume, PrimaryOperand=@WavePlayer_0.Out Front Right)
   Multiply_3 = UE.Multiply:Audio by Float(AdditionalOperands=^MasterVolume, PrimaryOperand=@WavePlayer_0.Out Side Right)
   Multiply_4 = UE.Multiply:Audio by Float(AdditionalOperands=^MasterVolume, PrimaryOperand=@WavePlayer_0.Out Side Left)
   Multiply_5 = UE.Multiply:Audio by Float(AdditionalOperands=^MasterVolume, PrimaryOperand=@WavePlayer_0.Out Low Frequency)
  ]
  links [
   {dst="@UE.OutputFormat.5dot1.Audio:0", src=Multiply_0.Out}
   {dst="@UE.OutputFormat.5dot1.Audio:2", src=Multiply_1.Out}
   {dst="@UE.OutputFormat.5dot1.Audio:1", src=Multiply_2.Out}
   {dst="@UE.OutputFormat.5dot1.Audio:5", src=Multiply_3.Out}
   {dst="@UE.OutputFormat.5dot1.Audio:4", src=Multiply_4.Out}
   {dst="@UE.OutputFormat.5dot1.Audio:3", src=Multiply_5.Out}
   {dst="@UE.Source.OneShot.OnFinished", src=WavePlayer_0.On Finished}
  ]
 }]
properties {
 assetImportData=&self:AssetImportData
 bOverrideConcurrency=true
 numChannels=6
 outputFormat=FiveDotOne
 sampleRate=48000
 soundClassObject=&Fort_UI
 volume=1.25
 concurrencyOverrides {maxCount=1}
}
