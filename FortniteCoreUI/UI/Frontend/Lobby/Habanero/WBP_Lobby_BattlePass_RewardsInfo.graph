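# Lobby battle pass widget that briefly shows an animated "Rewards Available!" callout with a count of rewards ready to claim.
# SetShowRewardInfo stores the count and, if it rose or an animation is pending, waits briefly, then plays the show animation and sound.
# When CanShowAnimation is false it marks the animation pending, and SetCanShowAnimation retries it.
# A timer plays the hide animation, with a Gate controlling the final switch to Hidden.
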WidgetBlueprint /FortniteCoreUI/UI/Frontend/Lobby/Habanero/WBP_Lobby_BattlePass_RewardsInfo
animations [
 Anim_ShowRewardsInfo {
  binding_count=8
  length_seconds=0.4
  object=&self:Anim_ShowRewardsInfo
  sequence {
   bindings [
    Txt_Rewards {
     id=#1
     possessed_class=CommonTextBlock
     tracks [
      {
       type=MovieScene2DTransformTrack
       sections [{
         type=MovieScene2DTransformSection
         end_frame=24
         start_frame=0
         channels [
          Translation.X {type=MovieSceneScriptingFloatChannel, keys [[0,360],[24,0]]}
          Translation.Y {type=MovieSceneScriptingFloatChannel, const [0,2]}
          Scale.X {type=MovieSceneScriptingFloatChannel, keys [[0,1.1],[12,1.1],[24,1]]}
          Scale.Y {type=MovieSceneScriptingFloatChannel, keys [[0,1.1],[12,1.1],[24,1]]}
         ]
        }]
      }
      {
       type=MovieSceneFloatTrack
       display_name=Render Opacity
       sections [{type=MovieSceneFloatSection, end_frame=24, start_frame=0, channels [None {type=MovieSceneScriptingFloatChannel, keys [[0,0],[12,0],[24,1]]}]}]
      }
     ]
    }
    Available_Flare {
     id=#2
     possessed_class=Image
     tracks [
      {
       type=MovieSceneFloatTrack
       display_name=Render Opacity
       sections [{type=MovieSceneFloatSection, end_frame=18, start_frame=0, channels [None {type=MovieSceneScriptingFloatChannel, keys [[0,0],[15,0],[18,1]]}]}]
      }
      {
       type=MovieSceneWidgetMaterialTrack
       sections [{
         type=MovieSceneParameterSection
         end_frame=24
         start_frame=0
         channels [FlareOpacity {type=MovieSceneScriptingFloatChannel, keys [[0,1],[18,1],[24,0.25]]}]
        }]
      }
      {
       type=MovieScene2DTransformTrack
       sections [{
         type=MovieScene2DTransformSection
         end_frame=24
         start_frame=0
         channels [
          Translation.X {type=MovieSceneScriptingFloatChannel, keys [[0,0],[12,180],[24,0]]}
          Translation.Y {type=MovieSceneScriptingFloatChannel, const [0,2]}
         ]
        }]
      }
     ]
    }
    Aura {
     id=#3
     possessed_class=Image
     tracks [
      {
       type=MovieSceneFloatTrack
       display_name=Render Opacity
       sections [{type=MovieSceneFloatSection, end_frame=21, start_frame=0, channels [None {type=MovieSceneScriptingFloatChannel, keys [[0,0],[6,0],[21,1]]}]}]
      }
      {
       type=MovieScene2DTransformTrack
       sections [{
         type=MovieScene2DTransformSection
         end_frame=24
         start_frame=0
         channels [Scale.X {type=MovieSceneScriptingFloatChannel, keys [[0,3],[6,3],[24,1]]}]
        }]
      }
     ]
    }
    Rewards_Flare {
     id=#4
     possessed_class=Image
     tracks [
      {
       type=MovieSceneFloatTrack
       display_name=Render Opacity
       sections [{type=MovieSceneFloatSection, end_frame=24, start_frame=0, channels [None {type=MovieSceneScriptingFloatChannel, keys [[0,0],[18,0],[24,1]]}]}]
      }
      {
       type=MovieSceneWidgetMaterialTrack
       sections [{
         type=MovieSceneParameterSection
         end_frame=24
         start_frame=0
         channels [FlareOpacity {type=MovieSceneScriptingFloatChannel, keys [[0,1],[21,1],[24,0.25]]}]
        }]
      }
      {
       type=MovieScene2DTransformTrack
       sections [{
         type=MovieScene2DTransformSection
         end_frame=24
         start_frame=0
         channels [
          Translation.X {type=MovieSceneScriptingFloatChannel, keys [[0,0],[12,180],[24,0]]}
          Translation.Y {type=MovieSceneScriptingFloatChannel, const [0,2]}
         ]
        }]
      }
     ]
    }
    WBP_UIKit_Highlight {
     id=#5
     possessed_class=WBP_UIKit_Highlight_C
     tracks [
      {
       type=MovieScene2DTransformTrack
       sections [{
         type=MovieScene2DTransformSection
         end_frame=24
         start_frame=0
         channels [
          Translation.X {type=MovieSceneScriptingFloatChannel, const [0,2]}
          Translation.Y {type=MovieSceneScriptingFloatChannel, const [0,2]}
          Scale.X {type=MovieSceneScriptingFloatChannel, keys [[0,1.2],[12,1.2],[24,1]]}
          Scale.Y {type=MovieSceneScriptingFloatChannel, keys [[0,1.2],[12,1.2],[24,1]]}
         ]
        }]
      }
      {
       type=MovieSceneFloatTrack
       display_name=Render Opacity
       sections [{
         type=MovieSceneFloatSection
         end_frame=24
         start_frame=0
         channels [None {type=MovieSceneScriptingFloatChannel, keys [[0,0],[15.6,0],[24,1]]}]
        }]
      }
     ]
    }
    Txt_Available {
     id=#6
     possessed_class=CommonTextBlock
     tracks [
      {
       type=MovieScene2DTransformTrack
       sections [{
         type=MovieScene2DTransformSection
         end_frame=24
         start_frame=0
         channels [
          Translation.X {type=MovieSceneScriptingFloatChannel, keys [[0,360],[24,0]]}
          Translation.Y {type=MovieSceneScriptingFloatChannel, const [0,2]}
          Scale.X {type=MovieSceneScriptingFloatChannel, keys [[0,1.1],[12,1.1],[24,1]]}
          Scale.Y {type=MovieSceneScriptingFloatChannel, keys [[0,1.1],[12,1.1],[24,1]]}
         ]
        }]
      }
      {
       type=MovieSceneFloatTrack
       display_name=Render Opacity
       sections [{type=MovieSceneFloatSection, end_frame=24, start_frame=0, channels [None {type=MovieSceneScriptingFloatChannel, keys [[0,0],[6,0],[24,1]]}]}]
      }
     ]
    }
    Particles {
     id=#7
     possessed_class=Image
     tracks [{
       type=MovieSceneWidgetMaterialTrack
       sections [{
         type=MovieSceneParameterSection
         end_frame=24
         start_frame=0
         channels [
          BigTrianglesOffset {type=MovieSceneScriptingFloatChannel, keys [[0,0],[15,3]]}
          Opacity {type=MovieSceneScriptingFloatChannel, keys [[0,0],[9,0],[21,1]]}
          SmallTrianglesOffset {type=MovieSceneScriptingFloatChannel, keys [[0,0],[24,1]]}
         ]
        }]
      }]
    }
    Glow {
     id=#8
     possessed_class=Image
     tracks [{
       type=MovieSceneFloatTrack
       display_name=Render Opacity
       sections [{type=MovieSceneFloatSection, end_frame=18, start_frame=0, channels [None {type=MovieSceneScriptingFloatChannel, keys [[0,0],[18,1]]}]}]
      }]
    }
   ]
   display_rate {numerator=60}
  }
 }
 Anim_HideRewardsInfo {
  binding_count=8
  length_seconds=0.3
  object=&self:Anim_HideRewardsInfo
  sequence {
   bindings [
    WBP_UIKit_Highlight {
     id=#9
     possessed_class=WBP_UIKit_Highlight_C
     tracks [
      {
       type=MovieScene2DTransformTrack
       sections [{
         type=MovieScene2DTransformSection
         end_frame=18
         start_frame=0
         channels [Translation.X {type=MovieSceneScriptingFloatChannel, keys [[0,0],[18,512]]}]
        }]
      }
      {
       type=MovieSceneFloatTrack
       display_name=Render Opacity
       sections [{type=MovieSceneFloatSection, channels [None {type=MovieSceneScriptingFloatChannel, keys [[0,1],[7,0]]}]}]
      }
     ]
    }
    Txt_Available {
     id=#10
     possessed_class=CommonTextBlock
     tracks [
      {
       type=MovieScene2DTransformTrack
       sections [{
         type=MovieScene2DTransformSection
         end_frame=18
         start_frame=0
         channels [Translation.X {type=MovieSceneScriptingFloatChannel, keys [[0,0],[18,768]]}]
        }]
      }
      {
       type=MovieSceneFloatTrack
       display_name=Render Opacity
       sections [{type=MovieSceneFloatSection, end_frame=18, start_frame=0, channels [None {type=MovieSceneScriptingFloatChannel, keys [[0,1],[7,0]]}]}]
      }
     ]
    }
    Aura {
     id=#11
     possessed_class=Image
     tracks [{
       type=MovieSceneFloatTrack
       display_name=Render Opacity
       sections [{type=MovieSceneFloatSection, channels [None {type=MovieSceneScriptingFloatChannel, keys [[0,1],[12,0]]}]}]
      }]
    }
    Glow {
     id=#12
     possessed_class=Image
     tracks [{
       type=MovieSceneFloatTrack
       display_name=Render Opacity
       sections [{type=MovieSceneFloatSection, channels [None {type=MovieSceneScriptingFloatChannel, keys [[0,1],[6,0]]}]}]
      }]
    }
    Particles {
     id=#13
     possessed_class=Image
     tracks [{
       type=MovieSceneWidgetMaterialTrack
       sections [{type=MovieSceneParameterSection, channels [Opacity {type=MovieSceneScriptingFloatChannel, keys [[0,1],[9,0]]}]}]
      }]
    }
    Rewards_Flare {
     id=#14
     possessed_class=Image
     tracks [
      {
       type=MovieSceneFloatTrack
       display_name=Render Opacity
       sections [{type=MovieSceneFloatSection, channels [None {type=MovieSceneScriptingFloatChannel, keys [[0,1],[6,0]]}]}]
      }
      {
       type=MovieSceneWidgetMaterialTrack
       sections [{type=MovieSceneParameterSection, channels [FlareOpacity {type=MovieSceneScriptingFloatChannel, const [0.25,1]}]}]
      }
     ]
    }
    Available_Flare {
     id=#15
     possessed_class=Image
     tracks [
      {
       type=MovieSceneFloatTrack
       display_name=Render Opacity
       sections [{type=MovieSceneFloatSection, channels [None {type=MovieSceneScriptingFloatChannel, keys [[0,1],[6,0]]}]}]
      }
      {
       type=MovieSceneWidgetMaterialTrack
       sections [{type=MovieSceneParameterSection, channels [FlareOpacity {type=MovieSceneScriptingFloatChannel, const [0.25,1]}]}]
      }
     ]
    }
    Txt_Rewards {
     id=#16
     possessed_class=CommonTextBlock
     tracks [
      {
       type=MovieScene2DTransformTrack
       sections [{
         type=MovieScene2DTransformSection
         end_frame=18
         start_frame=0
         channels [Translation.X {type=MovieSceneScriptingFloatChannel, keys [[0,0],[18,1024]]}]
        }]
      }
      {
       type=MovieSceneFloatTrack
       display_name=Render Opacity
       sections [{type=MovieSceneFloatSection, channels [None {type=MovieSceneScriptingFloatChannel, keys [[0,1],[7,0]]}]}]
      }
     ]
    }
   ]
   display_rate {numerator=60}
  }
 }
]
graphs [
 EventGraph {nodes [
   n1 = SetShowRewardInfo!() -> (OutputDelegate, then, RewardCount:int) {is=title K2Node_CustomEvent}
   > Set RewardsIncreased(RewardsIncreased:bool="integer > integer"(A=@n1.RewardCount, B=Get Reward Count() -> (Reward Count:int=0))) -> (Output_Get:bool)
   > n2 = Set Reward Count(Reward Count:int=@n1.RewardCount) -> (Output_Get:int=0)
   > StopAnimation(InAnimation=Get Anim_HideRewardsInfo() -> (Anim_HideRewardsInfo:Widget Animation Object Reference))
   > SetText!(
    execute
    self:WBP UIKit Highlight Base Object Reference=Get WBP_UIKit_Highlight() -> (WBP_UIKit_Highlight:WBP UIKit Highlight Object Reference)
    Text:Text=Format Text!(Format="LOCTEXT(\"{RewardCount} ready to claim\")", RewardCount:int=@n2) -> (Result)
   ) -> (then)
   > Branch(Condition=OR Boolean(A=Get RewardsIncreased() -> (RewardsIncreased:bool), B=Get IsPendingAnimation() -> (IsPendingAnimation:bool)))
   > Delay(Duration=1.5)
   > n3 = Branch(Condition=Get CanShowAnimation() -> (CanShowAnimation:bool))
   > Set Visibility!(execute, Visibility=HitTestInvisible) -> (then, Output_Get)
   > PlayAnimation(InAnimation=Get Anim_ShowRewardsInfo() -> (Anim_ShowRewardsInfo:Widget Animation Object Reference))
   > n4 = Branch(Condition=IsVisible())
   > n5 = PlaySound2D(Sound=/Game/Sounds/Fort_UI_Sounds/FrontEnd/New_FrontEnd/MSS_Fort_UI_FrontEnd_RewardsAvailable)
   Set IsPendingAnimation(execute=[@n5.then, @n4.else], IsPendingAnimation:bool=false) -> (Output_Get:bool)
   > StartAnimTimer()
   Set IsPendingAnimation(execute=@n3.else, IsPendingAnimation=true)
   StartAnimTimer() {is=title K2Node_CustomEvent}
   > n6 = Sequence()
   Delay(execute=@n6.then_1, Duration=Get DisplayLength() -> (DisplayLength:float))
   > n7 = Sequence()
   Gate!(Enter:Exec=@n8.Finished, Open:Exec=@n7.then_0, Close:Exec=@n6.then_0, Toggle:Exec, bStartClosed:bool=true) -> (Exit:Exec) {is=title K2Node_MacroInstance}
   > Set Visibility(Visibility=Hidden)
   HideRewardInfo(execute=@n7.then_1)
   HideRewardInfo() {is=title K2Node_CustomEvent}
   > StopAnimation(InAnimation=Get Anim_ShowRewardsInfo())
   > n8 = Play Animation with Finished event!(
    execute:Exec
    Widget:User Widget Object Reference=Self-Reference()
    InAnimation:Widget Animation Object Reference=Get Anim_HideRewardsInfo()
    StartAtTime:float=0
    NumLoopsToPlay:int=1
    PlayMode:EUMGSequencePlayMode Enum
    PlaybackSpeed:float=1
   ) -> (then:Exec, Result:Widget Animation Handle Structure, Finished:Exec) {is=title K2Node_PlayAnimation2}
  ]}
 SetCanShowAnimation {nodes [
   n1 = SetCanShowAnimation!() -> (then, CanShowAnimation:bool=~) {is=title K2Node_FunctionEntry}
   > Set CanShowAnimation(CanShowAnimation:bool=@n1) -> (Output_Get:bool)
   > Branch(Condition=Get IsPendingAnimation())
   > SetShowRewardInfo!(execute, self, RewardCount:int=Get Reward Count() -> (Reward Count=0)) -> (then) {is=title K2Node_CallFunction}
  ]}
]
variables [
 CanShowAnimation {type=boolean}
 DisplayLength {type=number, default=8}
 IsPendingAnimation {type=boolean}
 Reward Count {type=integer}
 RewardsIncreased {type=boolean}
]
widgets {
 info {inheritedWidgetCount=0, parentClass=&UserWidget, rootWidgetClass=&Overlay, widgetCount=14}
 widgets [
  Overlay_0 {widgetClassPath=&Overlay, widget=Overlay_0, properties {visibility=HitTestInvisible}}
  Glow {
   widgetClassPath=&Image
   parent=Overlay_0
   widget=Glow
   properties {
    visibility=SelfHitTestInvisible
    brush {
     bIsDynamicallyLoaded=false
     drawAs=Image
     imageType=FullColor
     mirroring=NoMirror
     resourceName=None
     resourceObject=/FortniteCoreUI/UI/Frontend/Lobby/Habanero/MaterialInstances/MI_UI_Lobby_RewardGlow
     tiling=NoTile
     imageSize=<32 32>
     margin=<ltrb 0 0 0 0>
     outlineSettings {
      bUseBrushTransparency=false
      roundingType=HalfHeightRadius
      width=0
      color {colorUseRule=UseColor_Specified, specifiedColor=<rgba 0 0 0 0>}
      cornerRadii=<0 0 0 0>
     }
     tintColor {colorUseRule=UseColor_Specified, specifiedColor=<rgba 1 1 1 1>}
     uVRegion {bIsValid=false, max=<0 0>, min=<0 0>}
    }
   }
   slot {class=OverlaySlot, properties {horizontalAlignment=HAlign_Fill, verticalAlignment=VAlign_Fill, padding=<ltrb -96 0 -96 0>}}
  }
  CommonVisualAttachment_0 {
   widgetClassPath=&CommonVisualAttachment
   parent=Overlay_0
   widget=CommonVisualAttachment_0
   properties {
    bOverride_HeightOverride=true
    bOverride_WidthOverride=true
    heightOverride=128
    widthOverride=576
    contentAnchor=<1 0.5>
    renderTransform {angle=0, scale=<1 1>, shear=<0 0>, translation=<216 0>}
   }
   slot {class=OverlaySlot, properties {horizontalAlignment=HAlign_Right, verticalAlignment=VAlign_Center}}
  }
  Overlay_Aura {widgetClassPath=&Overlay, bIsVariable=true, parent=CommonVisualAttachment_0, widget=Overlay_Aura, slot {class=SizeBoxSlot}}
  Aura {
   widgetClassPath=&Image
   parent=Overlay_Aura
   widget=Aura
   properties {
    visibility=SelfHitTestInvisible
    brush {
     bIsDynamicallyLoaded=false
     drawAs=Image
     imageType=FullColor
     mirroring=NoMirror
     resourceName=None
     resourceObject=/FortniteCoreUI/UI/Frontend/Lobby/Habanero/Materials/MI_UI_Lobby_RewardFlare
     tiling=NoTile
     imageSize=<332 32>
     margin=<ltrb 0 0 0 0>
     outlineSettings {
      bUseBrushTransparency=false
      roundingType=HalfHeightRadius
      width=0
      color {colorUseRule=UseColor_Specified, specifiedColor=<rgba 0 0 0 0>}
      cornerRadii=<0 0 0 0>
     }
     tintColor {colorUseRule=UseColor_Specified, specifiedColor=<rgba 1 1 1 1>}
     uVRegion {bIsValid=false, max=<0 0>, min=<0 0>}
    }
    renderTransformPivot=<1 0.5>
   }
   slot {class=OverlaySlot, properties {horizontalAlignment=HAlign_Right, verticalAlignment=VAlign_Fill}}
  }
  Particles {
   widgetClassPath=&Image
   parent=Overlay_Aura
   widget=Particles
   properties {
    visibility=SelfHitTestInvisible
    brush {
     bIsDynamicallyLoaded=false
     drawAs=Image
     imageType=FullColor
     mirroring=NoMirror
     resourceName=None
     resourceObject=/FortniteCoreUI/UI/Frontend/Lobby/Habanero/MaterialInstances/MI_UI_Lobby_Level_Particles
     tiling=NoTile
     imageSize=<32 32>
     margin=<ltrb 0 0 0 0>
     outlineSettings {
      bUseBrushTransparency=false
      roundingType=HalfHeightRadius
      width=0
      color {colorUseRule=UseColor_Specified, specifiedColor=<rgba 0 0 0 0>}
      cornerRadii=<0 0 0 0>
     }
     tintColor {colorUseRule=UseColor_Specified, specifiedColor=<rgba 1 1 1 1>}
     uVRegion {bIsValid=false, max=<0 0>, min=<0 0>}
    }
   }
   slot {class=OverlaySlot, properties {horizontalAlignment=HAlign_Fill, verticalAlignment=VAlign_Fill}}
  }
  VB_RewardsInfo {
   widgetClassPath=&VerticalBox
   bIsVariable=true
   parent=Overlay_0
   widget=VB_RewardsInfo
   slot {class=OverlaySlot, properties {horizontalAlignment=HAlign_Center, verticalAlignment=VAlign_Center}}
  }
  Overlay_1 {
   widgetClassPath=&Overlay
   parent=VB_RewardsInfo
   widget=Overlay_1
   slot {class=VerticalBoxSlot, properties {horizontalAlignment=HAlign_Center, verticalAlignment=VAlign_Bottom}}
  }
  Rewards_Flare {
   widgetClassPath=&Image
   parent=Overlay_1
   widget=Rewards_Flare
   properties {
    visibility=SelfHitTestInvisible
    brush {
     bIsDynamicallyLoaded=false
     drawAs=Image
     imageType=FullColor
     mirroring=NoMirror
     resourceName=None
     resourceObject=/FortniteCoreUI/UI/Frontend/Lobby/Habanero/MaterialInstances/MI_UI_Lobby_Level_TextFlare
     tiling=NoTile
     imageSize=<32 2>
     margin=<ltrb 0 0 0 0>
     outlineSettings {
      bUseBrushTransparency=false
      roundingType=HalfHeightRadius
      width=0
      color {colorUseRule=UseColor_Specified, specifiedColor=<rgba 0 0 0 0>}
      cornerRadii=<0 0 0 0>
     }
     tintColor {colorUseRule=UseColor_Specified, specifiedColor=<rgba 1 1 1 1>}
     uVRegion {bIsValid=false, max=<0 0>, min=<0 0>}
    }
   }
   slot {class=OverlaySlot, properties {horizontalAlignment=HAlign_Fill, verticalAlignment=VAlign_Fill, padding=<ltrb -192 4 -192 4>}}
  }
  Txt_Rewards {
   widgetClassPath=&CommonTextBlock
   parent=Overlay_1
   widget=Txt_Rewards
   properties {
    applyLineHeightToBottomLine=false
    justification=Center
    renderOpacity=0.7
    text=Rewards
    textTransformPolicy=ToUpper
    font {
     bForceMonospaced=false
     bMaterialIsStencil=false
     fontMaterial=null
     fontObject=&HeadingNow
     letterSpacing=0
     monospacedWidth=1
     size=15
     skewAmount=0
     typefaceFontName=86 Bold Italic
     outlineSettings {
      bApplyOutlineToDropShadows=false
      bMiteredCorners=false
      bSeparateFillAlpha=false
      outlineMaterial=null
      outlineSize=0
      outlineColor=<rgba 0 0 0 1>
     }
    }
    shadowOffset=<0 0>
   }
   slot {class=OverlaySlot, properties {horizontalAlignment=HAlign_Center, verticalAlignment=VAlign_Center}}
  }
  Overlay_2 {
   widgetClassPath=&Overlay
   parent=VB_RewardsInfo
   widget=Overlay_2
   slot {class=VerticalBoxSlot, properties {horizontalAlignment=HAlign_Center, verticalAlignment=VAlign_Bottom}}
  }
  Available_Flare {
   widgetClassPath=&Image
   parent=Overlay_2
   widget=Available_Flare
   properties {
    visibility=SelfHitTestInvisible
    brush {
     bIsDynamicallyLoaded=false
     drawAs=Image
     imageType=FullColor
     mirroring=NoMirror
     resourceName=None
     resourceObject=&MI_UI_Lobby_Level_TextFlare
     tiling=NoTile
     imageSize=<32 2>
     margin=<ltrb 0 0 0 0>
     outlineSettings {
      bUseBrushTransparency=false
      roundingType=HalfHeightRadius
      width=0
      color {colorUseRule=UseColor_Specified, specifiedColor=<rgba 0 0 0 0>}
      cornerRadii=<0 0 0 0>
     }
     tintColor {colorUseRule=UseColor_Specified, specifiedColor=<rgba 1 1 1 1>}
     uVRegion {bIsValid=false, max=<0 0>, min=<0 0>}
    }
   }
   slot {class=OverlaySlot, properties {horizontalAlignment=HAlign_Fill, verticalAlignment=VAlign_Fill, padding=<ltrb -256 10 -256 10>}}
  }
  Txt_Available {
   widgetClassPath=&CommonTextBlock
   parent=Overlay_2
   widget=Txt_Available
   properties {
    justification=Center
    text="Available!"
    font {
     bForceMonospaced=false
     bMaterialIsStencil=false
     fontMaterial=null
     fontObject=&HeadingNow
     letterSpacing=-10
     monospacedWidth=1
     size=25.5
     skewAmount=0
     typefaceFontName=86 Bold Italic
     outlineSettings {
      bApplyOutlineToDropShadows=false
      bMiteredCorners=false
      bSeparateFillAlpha=false
      outlineMaterial=null
      outlineSize=0
      outlineColor=<rgba 0 0 0 1>
     }
    }
    shadowOffset=<0 0>
   }
   slot {class=OverlaySlot, properties {horizontalAlignment=HAlign_Center, verticalAlignment=VAlign_Center}}
  }
  WBP_UIKit_Highlight {
   widgetClassPath=/FortUILibrary/UIKit/Highlight/WBP_UIKit_Highlight
   bIsVariable=true
   parent=VB_RewardsInfo
   widget=WBP_UIKit_Highlight
   properties {highlightSize=Compact, highlightText="## ready to claim", highlightType=Attention}
   slot {class=VerticalBoxSlot, properties {horizontalAlignment=HAlign_Center, verticalAlignment=VAlign_Top, padding=<ltrb 0 4 0 0>}}
  }
 ]
}
