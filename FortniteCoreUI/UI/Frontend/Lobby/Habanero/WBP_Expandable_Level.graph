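# Lobby Battle Pass widget on WBP_Expandable_Base showing level, XP and claimable rewards, opening the Battle Pass when clicked.
# Helper functions drive the progress ring's material parameter and toggle reward glow, overlay and info popup by reward count.
# Construct and input changes choose expanded state without level info for compact touch UI, otherwise collapsed extension with level info.
# The rewards popup's padding adjusts for touch input and safe zones.
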WidgetBlueprint /FortniteCoreUI/UI/Frontend/Lobby/Habanero/WBP_Expandable_Level
graphs [
 CollapseExtPart {nodes [
   CollapseExtPart() {is=title K2Node_FunctionEntry}
   > Parent: CollapseExtPart() {is=title K2Node_CallParentFunction}
   > SetVisibility(self=Get IndicatorWrapper() -> (IndicatorWrapper:Overlay Object Reference), InVisibility=SelfHitTestInvisible)
   > SetVisibility(
    self=[Get RewardOverlay() -> (RewardOverlay:Image Object Reference), Get RewardGlow() -> (RewardGlow:Image Object Reference)]
    InVisibility=Hidden
   )
  ]}
 EventGraph {nodes [
   n1 = OnVisibilityChanged_Event!() -> (OutputDelegate, then, InVisibility:ESlateVisibility Enum=Visible) {is=title K2Node_CustomEvent}
   > SetCanShowAnimation!(
    execute
    self:WBP Lobby Battle Pass Rewards Info Object Reference=Get WBP_Lobby_BattlePass_RewardsInfo() -> (WBP_Lobby_BattlePass_RewardsInfo:WBP Lobby Battle Pass Rewards Info Object Reference)
    CanShowAnimation:bool=IsVisible()
   ) -> (then)
   n2 = Event PreConstruct()
   > Parent: PreConstruct(IsDesignTime=@n2.IsDesignTime)
   > n3 = GetDynamicMaterial(self=Get RewardOverlay())
   > SetScalarParameterValue(self=@n3, ParameterName=CornerRadius, Value=20)
   Event Construct()
   > Assign On Visibility Changed!(execute:Exec, self:User Widget Object Reference, Delegate:"Delegate (by ref)"=@n1) -> (then:Exec) {is=title K2Node_AssignDelegate}
   > SetRewardsAuraSafeZone()
   > n4 = Set CollapseException(CollapseException:bool=IsUsingCompactUI()) -> (Output_Get:bool)
   n5 = GetOwningPlayer()
   n6 = Branch(
    execute=[@n4.then, @n8.then]
    Condition=AND Boolean(
     A=@n4
     B=IsUsingInputType!(InputType:ECommonInputType Enum=Touch, PlayerController:Player Controller Object Reference=@n5) -> (ReturnValue:bool)
    )
   )
   > SetExpanded()
   SetVisibility(
    execute=@n6.else
    self=[Get Separator() -> (Separator:Image Object Reference), Get VerticalBox_LevelInfo() -> (VerticalBox_LevelInfo:Vertical Box Object Reference)]
    InVisibility=SelfHitTestInvisible
   )
   > CollapseExtPart(self:WBP Expandable Base Object Reference) {is=title K2Node_CallFunction}
   Event On Clicked()
   > Parent: On Clicked() {is=title K2Node_CallParentFunction}
   > Fire Event Element Interacted(Element=Self-Reference())
   > NavigateToFeature!(execute, Feature:EFortUIFeature Enum=ShowBattlePass, bClearPrimaryContentStack:bool=true) -> (then)
   n7 = Event On Input Method Changed!() -> (OutputDelegate, then, CurrentInputType:ECommonInputType Enum) {is=title K2Node_Event}
   > n8 = SetRewardsTouchPadding!(execute, self, CurrentInput:ECommonInputType Enum=@n7.CurrentInputType) -> (then)
  ]}
 ExpandExtPart {nodes [
   ExpandExtPart() {is=title K2Node_FunctionEntry}
   > Parent: ExpandExtPart() {is=title K2Node_CallParentFunction}
   > SetVisibility(self=Get IndicatorWrapper(), InVisibility=Collapsed)
   > HideRewardInfo(self:WBP Lobby Battle Pass Rewards Info Object Reference=Get WBP_Lobby_BattlePass_RewardsInfo())
  ]}
 GetAttentionType {nodes [
   n1 = GetAttentionType!() -> (then, Enabled:bool=~) {is=title K2Node_FunctionEntry}
   > Return Node!(
    execute
    Result:E_UI_HighlightType Enum=Select(Option 0:E_UI_HighlightType Enum=NewEnumerator5, Option 1:E_UI_HighlightType Enum=NewEnumerator0, Index=@n1) -> (ReturnValue:E_UI_HighlightType Enum)
   )
  ]}
 GetBangVM {nodes [
   n1 = GetBangVM!() -> (then, PlayerVM:Fort Local Player VM Object Reference) {is=title K2Node_FunctionEntry}
   > n2 = Is Valid(InputObject=@n1)
   Return Node(execute=@n2.Is Not Valid, Result:Fort Local Player Bang VM Object Reference)
   > n3 = GetBangVM!(execute, self:Fort Local Player VM Object Reference=@n1, BangType:EFortBangType Enum=BattlePassTab) -> (then, ReturnValue:Fort Local Player Bang VM Object Reference) {is=title K2Node_CallFunction}
   > Return Node(Result=@n3)
  ]}
 ProgressXP {nodes [
   n1 = ProgressXP!() -> (then, CurrentXP:int, XPToNextLevel:int) {is=title K2Node_FunctionEntry}
   n2 = int + int(A=@n1, B=@n1.XPToNextLevel)
   > Return Node!(execute, ProgressXP:double=SafeDivide(A="To Float (Integer)"(InInt=@n1), B="To Float (Integer)"(InInt=@n2)))
  ]}
 SetExpanded {nodes [
   SetExpanded() {is=title K2Node_FunctionEntry}
   > SetVisibility(self=[Get Separator(), Get VerticalBox_LevelInfo()], InVisibility=Collapsed)
   > ExpandExtPart(self:WBP Expandable Base Object Reference) {is=title K2Node_CallFunction}
   > RequestTransition!(
    execute
    self:UIKit Block Object Reference=Get Block_Outline() -> (Block_Outline:WBP UIKit Block Outline Object Reference)
    State:EUIKitBlockVisualState Enum=Unfocused
   ) -> (then)
  ]}
 SetLevelCircleProgress {nodes [
   n1 = SetLevelCircleProgress!() -> (then, XPProgress:double=~) {is=title K2Node_FunctionEntry}
   > n2 = GetDynamicMaterial(self=Get LevelProgress() -> (LevelProgress:Image Object Reference))
   > SetScalarParameterValue(self=@n2, ParameterName=Get Parameter Name() -> (Parameter Name:Name=None), Value=@n1)
  ]}
 SetRewardsAuraSafeZone {nodes [
   SetRewardsAuraSafeZone() {is=title K2Node_FunctionEntry}
   > n1 = GetSafeZonePadding!(execute) -> (
    then
    SafePadding_X:double
    SafePadding_Y:double
    SafePadding_Z:double
    SafePadding_W:double
    SafePaddingScale:Vector 2D Structure
    SpillOverPadding:Vector 4 Structure
   )
   > Branch(Condition="float > float"(A=@n1.SafePadding_Z))
   n2 = Get WBP_Lobby_BattlePass_RewardsInfo()
   > SetPadding!(
    execute
    self:Size Box Slot Object Reference=SlotAsSizeBoxSlot!(Widget:Widget Object Reference=Get Overlay_Aura!(self:WBP Lobby Battle Pass Rewards Info Object Reference=@n2) -> (Overlay_Aura:Overlay Object Reference)) -> (ReturnValue:Size Box Slot Object Reference)
    InPadding_Left:float=0
    InPadding_Top:float=0
    InPadding_Right:float=-80
    InPadding_Bottom:float=0
   ) -> (then)
  ]}
 SetRewardsTouchPadding {nodes [
   n1 = SetRewardsTouchPadding!() -> (then, CurrentInput:ECommonInputType Enum) {is=title K2Node_FunctionEntry}
   n2 = Get WBP_Lobby_BattlePass_RewardsInfo()
   > SetPadding(
    self:Overlay Slot Object Reference=SlotAsOverlaySlot(Widget=Get VB_RewardsInfo!(self:WBP Lobby Battle Pass Rewards Info Object Reference=@n2) -> (VB_RewardsInfo:Vertical Box Object Reference))
    InPadding_Left=SelectFloat(A=-408, B=0, bPickA="Equal (Enum)"(A=@n1, B=Touch))
    InPadding_Top=0
    InPadding_Right=0
    InPadding_Bottom=0
   )
   > Return Node!(execute)
  ]}
 SetShowClaimableRewards {nodes [
   SetShowClaimableRewards!() -> (then, RewardCount:int) {is=title K2Node_FunctionEntry}
   n1 = "integer > integer"(A=Get RewardCount() -> (RewardCount:int=0))
   > SetVisibility(
    self=[Get RewardOverlay(), Get RewardGlow()]
    InVisibility=Select(Option 0:ESlateVisibility Enum=Hidden, Option 1:ESlateVisibility Enum=HitTestInvisible, Index=@n1) -> (ReturnValue:ESlateVisibility Enum)
   )
   > n2 = Branch(Condition=@n1)
   n3 = Get WBP_Lobby_BattlePass_RewardsInfo()
   > SetShowRewardInfo!(execute, self:WBP Lobby Battle Pass Rewards Info Object Reference=@n3, RewardCount:int=Get RewardCount() -> (RewardCount=0)) -> (then)
   HideRewardInfo(execute=@n2.else, self=@n3)
  ]}
]
variables [Parameter Name {type=string, default=Progress}]
widgets {
 info {
  inheritedWidgetCount=0
  namedSlotCount=4
  parentClass=/Game/UI/Frontend/Lobby/Habanero/WBP_Expandable_Base.WBP_Expandable_Base_C
  rootWidgetClass=null
  widgetCount=27
 }
 widgets [
  GridPanel_LevelBox {widgetClassPath=&GridPanel, widget=GridPanel_LevelBox, properties {columnFill [1], rowFill [1]}}
  FortVisualAttachment_0 {
   widgetClassPath=&FortVisualAttachment
   parent=GridPanel_LevelBox
   widget=FortVisualAttachment_0
   properties {bOverride_HeightOverride=true, bOverride_WidthOverride=true, heightOverride=144, widthOverride=144, contentAnchor=<0.5 0.5>}
   slot {class=GridSlot, properties {horizontalAlignment=HAlign_Center, verticalAlignment=VAlign_Center}}
  }
  LevelProgress {
   widgetClassPath=&Image
   bIsVariable=true
   parent=FortVisualAttachment_0
   widget=LevelProgress
   properties {
    visibility=SelfHitTestInvisible
    brush {
     bIsDynamicallyLoaded=false
     drawAs=Image
     imageType=FullColor
     mirroring=NoMirror
     resourceName=None
     resourceObject=/Game/UI/Frontend/Lobby/Habanero/Habanero_Materials/MaterialInstances/ProgressBars/MI_UI_Habanero_ProgressBar_PassLevel
     tiling=NoTile
     imageSize=<32 32>
     margin=<ltrb 0 0 0 0>
     outlineSettings {
      bUseBrushTransparency=false
      roundingType=HalfHeightRadius
      width=0
      color {colorUseRule=UseColor_Specified, specifiedColor=<rgba 0 0 0 0>}
      cornerRadii=<0 0 0 0>
     }
     tintColor {colorUseRule=UseColor_Specified, specifiedColor=<rgba 1 1 1 1>}
     uVRegion {bIsValid=false, max=<0 0>, min=<0 0>}
    }
   }
   slot {class=SizeBoxSlot}
  }
  VerticalBox_Level {
   widgetClassPath=&VerticalBox
   parent=GridPanel_LevelBox
   widget=VerticalBox_Level
   slot {class=GridSlot, properties {horizontalAlignment=HAlign_Center, layer=1, verticalAlignment=VAlign_Center}}
  }
  Text_2 {
   widgetClassPath=&CommonTextBlock
   parent=VerticalBox_Level
   widget=Text_2
   properties {
    bIsScrollingEnabled=false
    text=lvl
    textTransformPolicy=ToUpper
    colorAndOpacity {colorUseRule=UseColor_Specified, specifiedColor=<rgba 0.0048 0.004 0.0116 0.5>}
    font {
     bForceMonospaced=false
     bMaterialIsStencil=false
     fontMaterial=null
     fontObject=&HeadingNow
     letterSpacing=0
     monospacedWidth=1
     size=15.75
     skewAmount=0
     typefaceFontName=75 Medium
     outlineSettings {
      bApplyOutlineToDropShadows=false
      bMiteredCorners=false
      bSeparateFillAlpha=false
      outlineMaterial=null
      outlineSize=0
      outlineColor=<rgba 0 0 0 1>
     }
    }
    shadowOffset=<0 0>
    shapedTextOptions {bOverride_TextFlowDirection=true, bOverride_TextShapingMethod=false, textFlowDirection=Culture, textShapingMethod=Auto}
   }
   slot {class=VerticalBoxSlot, properties {horizontalAlignment=HAlign_Center, verticalAlignment=VAlign_Top, padding=<ltrb 0 0 0 -6>}}
  }
  Text_Level {
   widgetClassPath=&CommonTextBlock
   bIsVariable=true
   parent=VerticalBox_Level
   widget=Text_Level
   properties {
    bIsScrollingEnabled=false
    lineHeightPercentage=0.9
    mobileFontSizeMultiplier=1.25
    text=999
    colorAndOpacity {colorUseRule=UseColor_Specified, specifiedColor=<rgba 0.0048 0.004 0.0116 1>}
    font {
     bForceMonospaced=false
     bMaterialIsStencil=false
     fontMaterial=null
     fontObject=&HeadingNow
     letterSpacing=10
     monospacedWidth=1
     size=18
     skewAmount=0
     typefaceFontName=86 Bold
     outlineSettings {
      bApplyOutlineToDropShadows=false
      bMiteredCorners=false
      bSeparateFillAlpha=false
      outlineMaterial=null
      outlineSize=0
      outlineColor=<rgba 0 0 0 1>
     }
    }
    shadowOffset=<0 0>
    shapedTextOptions {bOverride_TextFlowDirection=true, bOverride_TextShapingMethod=false, textFlowDirection=Culture, textShapingMethod=Auto}
   }
   slot {class=VerticalBoxSlot, properties {horizontalAlignment=HAlign_Center, verticalAlignment=VAlign_Bottom, padding=<ltrb 0 -4 0 0>}}
  }
  IndicatorWrapper {
   widgetClassPath=&Overlay
   bIsVariable=true
   parent=GridPanel_LevelBox
   widget=IndicatorWrapper
   slot {class=GridSlot, properties {horizontalAlignment=HAlign_Right, verticalAlignment=VAlign_Top, padding=<ltrb 0 -3 -3 0>}}
  }
  WBP_UIKit_StatusIndicator {
   widgetClassPath=/FortUILibrary/UIKit/StatusIndicator/WBP_UIKit_StatusIndicator
   bIsVariable=true
   parent=IndicatorWrapper
   widget=WBP_UIKit_StatusIndicator
   properties {type=Attention, configuration=Dot, overrideStatusIndicatorVisibility=true, visibility=Collapsed}
   slot {class=OverlaySlot}
  }
  GridPanel_73 {widgetClassPath=&GridPanel, widget=GridPanel_73, properties {columnFill [1], rowFill [1]}}
  HorizontalBox_49 {
   widgetClassPath=&HorizontalBox
   parent=GridPanel_73
   widget=HorizontalBox_49
   slot {class=GridSlot, properties {padding=<ltrb 12 0 16 0>}}
  }
  HorizontalBox_52 {
   widgetClassPath=&HorizontalBox
   parent=HorizontalBox_49
   widget=HorizontalBox_52
   slot {class=HorizontalBoxSlot, properties {verticalAlignment=VAlign_Center, size {sizeRule=Fill, value=1}}}
  }
  SizeBox_0 {
   widgetClassPath=&SizeBox
   parent=HorizontalBox_52
   widget=SizeBox_0
   properties {bOverride_HeightOverride=true, bOverride_MinDesiredWidth=true, heightOverride=48, minDesiredWidth=48}
   slot {class=HorizontalBoxSlot, properties {horizontalAlignment=HAlign_Center, verticalAlignment=VAlign_Center}}
  }
  RewardAmount {
   widgetClassPath=/FortUILibrary/UIKit/Highlight/WBP_UIKit_Highlight
   bIsVariable=true
   parent=SizeBox_0
   widget=RewardAmount
   properties {highlightText=XX, highlightType=Attention}
   slot {class=SizeBoxSlot}
  }
  RewardAmount_Text {
   widgetClassPath=&CommonTextBlock
   bIsVariable=true
   parent=HorizontalBox_52
   widget=RewardAmount_Text
   properties {
    autoWrapText=true
    bIsScrollingEnabled=false
    mobileFontSizeMultiplier=1.25
    text=Rewards to claim
    textTransformPolicy=ToUpper
    wrapTextAt=150
    colorAndOpacity {colorUseRule=UseColor_Specified, specifiedColor=<rgba 0.0048 0.004 0.0116 1>}
    font {
     bForceMonospaced=false
     bMaterialIsStencil=false
     fontMaterial=null
     fontObject=&HeadingNow
     letterSpacing=0
     monospacedWidth=1
     size=22.5
     skewAmount=0
     typefaceFontName=75 Medium Italic
     outlineSettings {
      bApplyOutlineToDropShadows=false
      bMiteredCorners=false
      bSeparateFillAlpha=false
      outlineMaterial=null
      outlineSize=0
      outlineColor=<rgba 0 0 0 1>
     }
    }
    shadowOffset=<0 0>
    shapedTextOptions {bOverride_TextFlowDirection=true, bOverride_TextShapingMethod=false, textFlowDirection=Culture, textShapingMethod=Auto}
   }
   slot {class=HorizontalBoxSlot, properties {horizontalAlignment=HAlign_Center, verticalAlignment=VAlign_Center, padding=<ltrb 8 0 0 0>}}
  }
  Separator {
   widgetClassPath=&Image
   bIsVariable=true
   parent=HorizontalBox_49
   widget=Separator
   properties {
    renderOpacity=0.25
    visibility=SelfHitTestInvisible
    brush {
     bIsDynamicallyLoaded=false
     drawAs=Image
     imageType=NoImage
     mirroring=NoMirror
     resourceName=None
     resourceObject=null
     tiling=NoTile
     imageSize=<2 1>
     margin=<ltrb 0 0 0 0>
     outlineSettings {
      bUseBrushTransparency=false
      roundingType=HalfHeightRadius
      width=0
      color {colorUseRule=UseColor_Specified, specifiedColor=<rgba 0 0 0 0>}
      cornerRadii=<0 0 0 0>
     }
     tintColor {colorUseRule=UseColor_Specified, specifiedColor=<rgba 1 1 1 1>}
     uVRegion {bIsValid=false, max=<0 0>, min=<0 0>}
    }
   }
   slot {class=HorizontalBoxSlot, properties {horizontalAlignment=HAlign_Center, padding=<ltrb 8 8 8 8>}}
  }
  VerticalBox_LevelInfo {
   widgetClassPath=&VerticalBox
   bIsVariable=true
   parent=HorizontalBox_49
   widget=VerticalBox_LevelInfo
   slot {class=HorizontalBoxSlot, properties {verticalAlignment=VAlign_Center}}
  }
  HorizontalBox_51 {
   widgetClassPath=&HorizontalBox
   parent=VerticalBox_LevelInfo
   widget=HorizontalBox_51
   slot {class=VerticalBoxSlot, properties {verticalAlignment=VAlign_Center, size {sizeRule=Fill, value=1}}}
  }
  Text_XpToNextLevel {
   widgetClassPath=&CommonTextBlock
   bIsVariable=true
   parent=HorizontalBox_51
   widget=Text_XpToNextLevel
   properties {
    bIsScrollingEnabled=false
    mobileFontSizeMultiplier=1.25
    text=XXXXX
    textTransformPolicy=ToUpper
    colorAndOpacity {colorUseRule=UseColor_Specified, specifiedColor=<rgba 0.0048 0.004 0.0116 1>}
    font {
     bForceMonospaced=false
     bMaterialIsStencil=false
     fontMaterial=null
     fontObject=&HeadingNow
     letterSpacing=0
     monospacedWidth=1
     size=22.5
     skewAmount=0
     typefaceFontName=75 Medium Italic
     outlineSettings {
      bApplyOutlineToDropShadows=false
      bMiteredCorners=false
      bSeparateFillAlpha=false
      outlineMaterial=null
      outlineSize=0
      outlineColor=<rgba 0 0 0 1>
     }
    }
    shadowOffset=<0 0>
    shapedTextOptions {bOverride_TextFlowDirection=true, bOverride_TextShapingMethod=false, textFlowDirection=Culture, textShapingMethod=Auto}
   }
   slot {class=HorizontalBoxSlot, properties {padding=<ltrb 0 0 8 0>}}
  }
  Text_Level_3 {
   widgetClassPath=&CommonTextBlock
   parent=HorizontalBox_51
   widget=Text_Level_3
   properties {
    bIsScrollingEnabled=false
    mobileFontSizeMultiplier=1.25
    text=XP
    textTransformPolicy=ToUpper
    colorAndOpacity {colorUseRule=UseColor_Specified, specifiedColor=<rgba 0.0048 0.004 0.0116 1>}
    font {
     bForceMonospaced=false
     bMaterialIsStencil=false
     fontMaterial=null
     fontObject=&HeadingNow
     letterSpacing=0
     monospacedWidth=1
     size=22.5
     skewAmount=0
     typefaceFontName=75 Medium Italic
     outlineSettings {
      bApplyOutlineToDropShadows=false
      bMiteredCorners=false
      bSeparateFillAlpha=false
      outlineMaterial=null
      outlineSize=0
      outlineColor=<rgba 0 0 0 1>
     }
    }
    shadowOffset=<0 0>
    shapedTextOptions {bOverride_TextFlowDirection=true, bOverride_TextShapingMethod=false, textFlowDirection=Culture, textShapingMethod=Auto}
   }
   slot {class=HorizontalBoxSlot}
  }
  Overlay_0 {
   widgetClassPath=&Overlay
   parent=VerticalBox_LevelInfo
   widget=Overlay_0
   slot {class=VerticalBoxSlot, properties {verticalAlignment=VAlign_Center, size {sizeRule=Fill, value=1}}}
  }
  Text_Level_2 {
   widgetClassPath=&CommonTextBlock
   parent=Overlay_0
   widget=Text_Level_2
   properties {
    bIsScrollingEnabled=false
    mobileFontSizeMultiplier=1.25
    text=to next LVL
    textTransformPolicy=ToUpper
    colorAndOpacity {colorUseRule=UseColor_Specified, specifiedColor=<rgba 0.0666 0.0666 0.0666 1>}
    font {
     bForceMonospaced=false
     bMaterialIsStencil=false
     fontMaterial=null
     fontObject=&HeadingNow
     letterSpacing=0
     monospacedWidth=1
     size=18
     skewAmount=0
     typefaceFontName=75 Medium
     outlineSettings {
      bApplyOutlineToDropShadows=false
      bMiteredCorners=false
      bSeparateFillAlpha=false
      outlineMaterial=null
      outlineSize=0
      outlineColor=<rgba 0 0 0 1>
     }
    }
    shadowOffset=<0 0>
    shapedTextOptions {bOverride_TextFlowDirection=true, bOverride_TextShapingMethod=false, textFlowDirection=Culture, textShapingMethod=Auto}
   }
   slot {class=OverlaySlot}
  }
  Overlay_20 {widgetClassPath=&Overlay, widget=Overlay_20}
  RewardOverlay {
   widgetClassPath=&Image
   bIsVariable=true
   parent=Overlay_20
   widget=RewardOverlay
   properties {
    visibility=Hidden
    brush {
     bIsDynamicallyLoaded=false
     drawAs=Image
     imageType=FullColor
     mirroring=NoMirror
     resourceName=None
     resourceObject=/FortUILibrary/MaterialInstances/Effects/MI_UI_RewardState
     tiling=NoTile
     imageSize=<32 32>
     margin=<ltrb 0 0 0 0>
     outlineSettings {
      bUseBrushTransparency=false
      roundingType=HalfHeightRadius
      width=0
      color {colorUseRule=UseColor_Specified, specifiedColor=<rgba 0 0 0 0>}
      cornerRadii=<0 0 0 0>
     }
     tintColor {colorUseRule=UseColor_Specified, specifiedColor=<rgba 1 1 1 1>}
     uVRegion {bIsValid=false, max=<0 0>, min=<0 0>}
    }
   }
   slot {class=OverlaySlot, properties {horizontalAlignment=HAlign_Fill, verticalAlignment=VAlign_Fill}}
  }
  Overlay_59 {widgetClassPath=&Overlay, widget=Overlay_59}
  RewardGlow {
   widgetClassPath=&Image
   bIsVariable=true
   parent=Overlay_59
   widget=RewardGlow
   properties {
    visibility=Hidden
    brush {
     bIsDynamicallyLoaded=false
     drawAs=Image
     imageType=FullColor
     mirroring=NoMirror
     resourceName=None
     resourceObject=/FortUILibrary/MaterialInstances/Effects/MI_UI_RewardBacking
     tiling=NoTile
     imageSize=<32 32>
     margin=<ltrb 0 0 0 0>
     outlineSettings {
      bUseBrushTransparency=false
      roundingType=HalfHeightRadius
      width=0
      color {colorUseRule=UseColor_Specified, specifiedColor=<rgba 0 0 0 0>}
      cornerRadii=<0 0 0 0>
     }
     tintColor {colorUseRule=UseColor_Specified, specifiedColor=<rgba 1 1 1 1>}
     uVRegion {bIsValid=false, max=<0 0>, min=<0 0>}
    }
   }
   slot {class=OverlaySlot, properties {horizontalAlignment=HAlign_Fill, verticalAlignment=VAlign_Fill, padding=<ltrb -128 -64 -128 -64>}}
  }
  FortVisualAttachment_41 {
   widgetClassPath=&FortVisualAttachment
   parent=Overlay_59
   widget=FortVisualAttachment_41
   properties {contentAnchor=<1 0.5>}
   slot {class=OverlaySlot, properties {horizontalAlignment=HAlign_Right, verticalAlignment=VAlign_Center}}
  }
  WBP_Lobby_BattlePass_RewardsInfo {
   widgetClassPath=&/WBP_Lobby_BattlePass_RewardsInfo
   bIsVariable=true
   parent=FortVisualAttachment_41
   widget=WBP_Lobby_BattlePass_RewardsInfo
   properties {visibility=SelfHitTestInvisible}
   slot {class=SizeBoxSlot, properties {horizontalAlignment=HAlign_Right, verticalAlignment=VAlign_Center, padding=<ltrb 0 0 136 0>}}
  }
 ]
}
