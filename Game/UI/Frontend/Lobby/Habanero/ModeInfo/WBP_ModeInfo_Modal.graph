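# Lobby modal that presents Habanero (Ranked) mode info, built on FortHabaneroInfoModal.
# The event graph holds only unwired events, so the widget supplies layout for the parent's bound panels and buttons.
# A root grid holds the inherited scroll box and a vertical box filled by WBP_ModeIntfoList.
# A bottom-bar decoy button handles Close, while a zero-opacity Close button with ranked sounds binds HabaneroToggle over a post-buffer capture in a safe zone.
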WidgetBlueprint /Game/UI/Frontend/Lobby/Habanero/ModeInfo/WBP_ModeInfo_Modal
graphs [EventGraph {nodes [
   Event PreConstruct()
   Event Construct()
   Event Tick()
  ]}]
variables []
widgets {
 info {inheritedWidgetCount=4, parentClass=/Script/FortniteUI.FortHabaneroInfoModal, rootWidgetClass=&GridPanel, widgetCount=9}
 widgets [
  Root {widgetClassPath=&GridPanel, widget=Root, properties {columnFill [1], rowFill [1]}}
  ScrollBox_Panels {
   widgetClassPath=/Script/FortniteUI.AthenaScrollBox
   bInherited=true
   bIsVariable=true
   parent=Root
   widget=ScrollBox_Panels
   properties {
    allowOverscroll=false
    bAllowRightClickDragScrolling=false
    bAnimateWheelScrolling=false
    bEnableClickThrough=true
    consumeMouseWheel=Never
    scrollBarVisibility=Collapsed
    scrollWhenFocusChanges=AnimatedScroll
    visibility=SelfHitTestInvisible
   }
   slot {class=GridSlot}
  }
  Vbox_Panels {widgetClassPath=&VerticalBox, bInherited=true, parent=Root, widget=Vbox_Panels, slot {class=GridSlot}}
  WBP_ModeIntfoList {
   widgetClassPath=&/WBP_ModeIntfoList
   bIsVariable=true
   parent=Vbox_Panels
   widget=WBP_ModeIntfoList
   slot {class=VerticalBoxSlot, properties {size {sizeRule=Fill, value=1}}}
  }
  SafeZone_bottombuttons {
   widgetClassPath=&SafeZone
   bIsVariable=true
   parent=Root
   widget=SafeZone_bottombuttons
   properties {padLeft=false, padTop=false}
   slot {class=GridSlot, properties {horizontalAlignment=HAlign_Right, verticalAlignment=VAlign_Bottom, padding=<ltrb 0 0 36 32>}}
  }
  Overlay_3 {widgetClassPath=&Overlay, parent=SafeZone_bottombuttons, widget=Overlay_3, slot {class=SafeZoneSlot}}
  CaptureForPostBufferUpdate_159 {
   widgetClassPath=&WBP_CaptureForPostBufferUpdate
   parent=Overlay_3
   widget=CaptureForPostBufferUpdate_159
   slot {class=OverlaySlot, properties {horizontalAlignment=HAlign_Fill, verticalAlignment=VAlign_Fill}}
  }
  Button_Close_Alt {
   widgetClassPath=/FortUILibrary/UIKit/Buttons/WBP_UIKit_Button_Quiet
   bInherited=true
   bIsVariable=true
   parent=Overlay_3
   widget=Button_Close_Alt
   properties {
    bDisplayInputActionWhenNotInteractable=false
    bHideInputAction=true
    bIsFocusable=false
    renderOpacity=0
    text=Close
    hoveredSlateSoundOverride {resourceObject=/Game/Sounds/Fort_UI_Sounds/FrontEnd/New_FrontEnd/Ranked/Fortnite_UI_Ranked_Button_Hover_Cue}
    pressedSlateSoundOverride {resourceObject=/Game/Sounds/Fort_UI_Sounds/FrontEnd/New_FrontEnd/Ranked/Fortnite_UI_Ranked_Button_Back_Cue}
    selectedHoveredSlateSoundOverride {resourceObject=&Fortnite_UI_Ranked_Button_Hover_Cue}
    selectedPressedSlateSoundOverride {resourceObject=&Fortnite_UI_Ranked_Button_Back_Cue}
    triggeringInputAction {dataTable=/Game/Athena/UI/Frontend/Lobby/AthenaLobbyInputActionDataTable, rowName=HabaneroToggle}
   }
   slot {class=OverlaySlot}
  }
  Button_Close {
   widgetClassPath=/Game/UI/Frontend/BottomBar/WBP_BottomBarDecoyButton
   bInherited=true
   bIsVariable=true
   parent=Root
   widget=Button_Close
   properties {inputPriority=1001, priority=1001, triggeringInputAction {dataTable=&InputActionDataTable, rowName=Close}}
   slot {class=GridSlot, properties {horizontalAlignment=HAlign_Left, verticalAlignment=VAlign_Top}}
  }
 ]
}
